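# Sound cue for stage fireworks at the Papaya event, layering a one-shot launch with a looping fireworks bed.
# A mixer blends the loop at full level with the quieter launch, each shaped by its own volume envelope.
# The loop swells in over a few seconds then gradually settles low, while the launch starts full and decays.
# Output uses a dedicated stage fireworks attenuation and the live in-world sound class.
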SoundCue /Game/Athena/Sounds/Events/Papaya/SFX/Papaya_Stage_Fireworks_Cue
root=Mixer_0
nodes [
 {
  id=Enveloper_0
  children [WavePlayer_0]
  properties {pitchCurve {editorCurveData {keys [{time=0}]}}, volumeCurve {editorCurveData {keys [<at 0 0>,<at 3 1>,<at 10 0.3>,<at 25 0.15>]}}}
 }
 {
  id=Enveloper_1
  children [WavePlayer_1]
  properties {pitchCurve {editorCurveData {keys [{time=0}]}}, volumeCurve {editorCurveData {keys [<at 0 1>,<at 4 0.2>]}}}
 }
 {id=Mixer_0, children [Enveloper_0,Enveloper_1], properties {inputVolume [1,0.45]}}
 {id=WavePlayer_0, properties {bLooping=true, soundWaveAssetPtr=&/Papaya_Fireworks_Loop_02}}
 {id=WavePlayer_1, properties {soundWaveAssetPtr=&/Papaya_Fireworks_Launch_01}}
]
properties {
 attenuationSettings=&/Papaya_Stage_Fireworks_Att
 maxDistance=2097152
 soundClassObject=/Game/Sounds/FortSoundClassesAndModes/Fort_Live_InWorld
 volumeMultiplier=0.5
}
